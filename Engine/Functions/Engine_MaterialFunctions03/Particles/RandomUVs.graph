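# RandomUVs is a material function that shifts particle UV coordinates using per-particle DynamicParameter values.
# It adds an appended two-component offset to the Coordinate Input, which defaults to texture coordinates, and outputs the result.
# Both components read V_Offset, so U_Offset is declared but unused and the shift is always diagonal.
# Its description expects a uniform 0-1 distribution per channel, set at spawn time only.
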MaterialFunction /Engine/Functions/Engine_MaterialFunctions03/Particles/RandomUVs
graph {
 properties {
  description="Uses a DynamicParameter with U_Offset and V_Offset as parameter names to move UV coordinates around.  Set up as a uniform distribution, at spawn time only, with a 0-1 distribution per-channel."
  libraryCategoriesText [Misc,Particles]
 }
 nodes [
  n1 = FunctionInput(Preview=TextureCoordinate()) {bUsePreviewValueAsDefault=true, description="Default coordinates are ", inputName=Coordinate Input, inputType=FunctionInput_Vector2}
  n2 = DynamicParameter() {defaultValue=<rgba 0 0 0 1>, paramNames [U_Offset,V_Offset,Param3,Param4]}
  FunctionOutput(None=Add(A=@n1, B=AppendVector(A=@n2.V_Offset, B=@n2.V_Offset))) {description=2-vector UV coordinates offset by DynamicParameter, outputName=UV out}
 ]
}
