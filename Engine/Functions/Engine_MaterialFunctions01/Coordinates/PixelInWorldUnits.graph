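# Material function that converts a 2D pixel offset at a given depth into a world-space offset vector, intended for extruding particles by pixel sizes alongside DOFFunction.
# A Custom node reads the cotangent of half the horizontal FOV, which with half the view width divides Depth (PixelDepth by default) to give a pixel's view-space size.
# That size scales the Pixel input, is extended with a constant Z, and is transformed from view to world space.
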MaterialFunction /Engine/Functions/Engine_MaterialFunctions01/Coordinates/PixelInWorldUnits
graph {
 properties {bExposeToLibrary=true}
 nodes [
  n1 = FunctionInput(Preview=PixelDepth().R) {bUsePreviewValueAsDefault=true, desc=">=0", description=from camera e.g. PixelDepth, inputName=Depth, inputType=FunctionInput_Scalar}
  n2 = ComponentMask(None=Multiply(A=Custom!(None) {code="GetCotanHalfFieldOfView().x", desc="GetCotanHalfFieldOfView()", outputType=CMOT_Float2}, B=ViewSize())) {r=true}
  n3 = Multiply(
   A=FunctionInput() {description=offset, inputName=Pixel, inputType=FunctionInput_Vector2, previewValue=<1 1 0 1>}
   B=Divide(A=@n1, B=Multiply(A=@n2) {constB=0.5}) {desc=Size of a pixel in view space}
  )
  n4 = Constant()
  FunctionOutput(None=Transform(None=AppendVector(A=@n3, B=@n4)) {transformSourceType=TRANSFORMSOURCE_View}) {
   desc="Size of a pixel in world space, multiply with -1..1 to extrude a particle by a pixel size (use with DOFFunction)"
   outputName=PixelInWorldUnits
  }
 ]
}
