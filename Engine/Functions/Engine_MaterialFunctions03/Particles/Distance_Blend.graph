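# Material function in the Misc and Particles library that produces a 0 to 1 blend based on pixel depth.
# It adds a Start Offset input, meant as a negative distance in units, to PixelDepth, then divides by a Blend Range input.
# The clamped result is zero before the start distance and ramps linearly to one across the range, with preview values serving as defaults for both inputs.
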MaterialFunction /Engine/Functions/Engine_MaterialFunctions03/Particles/Distance_Blend
graph {
 properties {bExposeToLibrary=true, description=Sets a distance and range to blend from 0 to 1, libraryCategoriesText [Misc,Particles]}
 nodes [
  n1 = FunctionInput() {
   bUsePreviewValueAsDefault=true
   description=Input should be a negative value in units
   inputName=Start Offset
   inputType=FunctionInput_Scalar
   previewValue=<-1000 0 0 1>
  }
  n2 = Add(A=PixelDepth().R, B=@n1) {constB=-1000, desc=Sets Blend Start Distance}
  n3 = FunctionInput() {bUsePreviewValueAsDefault=true, inputName=Blend Range, inputType=FunctionInput_Scalar, previewValue=<100 0 0 1>}
  FunctionOutput(None=Clamp(None=Divide(A=@n2, B=@n3) {constB=100, desc=Sets Blend Range}))
 ]
}
